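# Physics asset for the M_MED_DriftSwat backpack cosmetic, setting up secondary motion for two knives and a right zipper.
# Each part pairs a base bone capsule with a dyn bone capsule, joined by one constraint.
# The knives get only a few degrees of swing while the zipper swings much more freely, and all three joints have a zero twist limit.
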PhysicsAsset /BRCosmetics/Accessories/FORT_Backpacks/Backpack_M_MED_DriftSwat/Meshes/M_MED_DriftSwat_Pack_Physics
bodies {
 base_knife_down [{shapeType=Capsule, length=0.505, radius=0.505, shapeName=base_knife_down_capsule, rotation=<rot 0 -90 0>}]
 base_knife_up [{shapeType=Capsule, length=0.505, radius=0.505, shapeName=base_knife_up_capsule, rotation=<rot 0 -90 0>}]
 base_zipper_r [{shapeType=Capsule, length=0.505, radius=0.505, shapeName=base_zipper_r_capsule, rotation=<rot 0 -90 0>}]
 dyn_knife_down [{shapeType=Capsule, length=2, radius=1, shapeName=dyn_knife_down_capsule, center=<1 0 0>, rotation=<rot 90 0 0>}]
 dyn_knife_up [{shapeType=Capsule, length=2, radius=1, shapeName=dyn_knife_up_capsule, center=<3.7 0 0>, rotation=<rot 90 0 0>}]
 dyn_zipper_r [{shapeType=Capsule, length=0.505, radius=0.505, shapeName=dyn_zipper_r_capsule, center=<1.6351 0 0>, rotation=<rot 90 0 0>}]
}
constraints [
 {bone1Name=dyn_knife_down, bone2Name=base_knife_down, swing1LimitDegrees=5, swing2LimitDegrees=1, twistLimitDegrees=0}
 {bone1Name=dyn_knife_up, bone2Name=base_knife_up, swing1LimitDegrees=5, swing2LimitDegrees=2, twistLimitDegrees=0}
 {bone1Name=dyn_zipper_r, bone2Name=base_zipper_r, swing1LimitDegrees=50, swing2LimitDegrees=30, twistLimitDegrees=0}
]
